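# Physics asset for the left-arm sleeve dynamics of the male medium Robot Trouble body cosmetic.
# Capsule bodies cover four dyn_sleeve bones, with smaller capsules on the clavicle, upper arm, forearm and twist bones.
# The top sleeve hangs off the clavicle, back and bottom off the upper arm, all with tight angular limits.
# The forearm sleeve attaches to the lower arm with looser limits, while constraints between arm bones rotate freely.
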PhysicsAsset /BRCosmetics/Characters/Player/Male/Medium/Bodies/M_MED_Robot_Trouble/Meshes/M_MED_Robot_Trouble_Dyn_Arm_l
bodies {
 clavicle_l [{shapeType=Capsule, length=1, radius=1}]
 dyn_sleeve_back_l [{shapeType=Capsule, length=3, radius=2, center=<0.1 0 0>, rotation=<rot 90 0 0>}]
 dyn_sleeve_bott_l [{shapeType=Capsule, length=3, radius=2, center=<0.1 0 0>, rotation=<rot 90 0 0>}]
 dyn_sleeve_forearm_l [{shapeType=Capsule, length=3, radius=2, rotation=<rot 90 0 0>}]
 dyn_sleeve_top_l [{shapeType=Capsule, length=3, radius=2, center=<0.1 0 0>, rotation=<rot 90 0 0>}]
 lowerarm_l [{shapeType=Capsule, length=1, radius=1}]
 lowerarm_twist_02_l [{shapeType=Capsule, length=1, radius=1}]
 upperarm_l [{shapeType=Capsule, length=1, radius=1}]
 upperarm_twist_01_l [{shapeType=Capsule, length=1, radius=1}]
}
constraints [
 {bone1Name=dyn_sleeve_forearm_l, bone2Name=lowerarm_l, swing1LimitDegrees=45, swing2LimitDegrees=45, twistLimitDegrees=45}
 {
  bone1Name=lowerarm_twist_02_l
  bone2Name=lowerarm_l
  swing1LimitDegrees=45
  swing1Motion=Free
  swing2LimitDegrees=45
  swing2Motion=Free
  twistLimitDegrees=45
  twistMotion=Free
 }
 {
  bone1Name=lowerarm_l
  bone2Name=upperarm_l
  swing1LimitDegrees=45
  swing1Motion=Free
  swing2LimitDegrees=45
  swing2Motion=Free
  twistLimitDegrees=45
  twistMotion=Free
 }
 {bone1Name=dyn_sleeve_bott_l, bone2Name=upperarm_l, swing1LimitDegrees=10, swing2LimitDegrees=20, twistLimitDegrees=2}
 {bone1Name=dyn_sleeve_back_l, bone2Name=upperarm_l, swing1LimitDegrees=10, swing2LimitDegrees=20, twistLimitDegrees=2}
 {bone1Name=dyn_sleeve_top_l, bone2Name=clavicle_l, swing1LimitDegrees=10, swing2LimitDegrees=20, twistLimitDegrees=2}
 {
  bone1Name=upperarm_l
  bone2Name=clavicle_l
  swing1LimitDegrees=45
  swing1Motion=Free
  swing2LimitDegrees=45
  swing2Motion=Free
  twistLimitDegrees=45
  twistMotion=Free
 }
 {
  bone1Name=upperarm_twist_01_l
  bone2Name=upperarm_l
  swing1LimitDegrees=45
  swing1Motion=Free
  swing2LimitDegrees=45
  swing2Motion=Free
  twistLimitDegrees=45
  twistMotion=Free
 }
]
