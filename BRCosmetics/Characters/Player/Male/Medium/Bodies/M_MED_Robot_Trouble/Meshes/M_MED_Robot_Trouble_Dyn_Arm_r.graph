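# Physics asset for the right-arm sleeve and glove cuff dyn bones on the M_MED_Robot_Trouble player body cosmetic.
# Small capsules on outer, top, front and back sleeve bones and the cuff hang from clavicle, upper-arm or forearm bodies with tight swing and twist limits.
# The inner bottom sleeve is a two-link box chain off upperarm_twist_02_r.
# Arm and hand capsules link to clavicle_r with free angular motion.
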PhysicsAsset /BRCosmetics/Characters/Player/Male/Medium/Bodies/M_MED_Robot_Trouble/Meshes/M_MED_Robot_Trouble_Dyn_Arm_r
bodies {
 clavicle_r [{shapeType=Capsule, length=1, radius=1}]
 dyn_gloove_cuff_r [{shapeType=Capsule, length=3, radius=2, center=<0.1 0 0>, rotation=<rot 90 0 0>}]
 dyn_sleeve_bott_bk_r [{shapeType=Capsule, length=3, radius=2, center=<0.3 0 0>, rotation=<rot 90 0 0>}]
 dyn_sleeve_bott_fr_r [{shapeType=Capsule, length=3, radius=2, center=<0.3 0 0>, rotation=<rot 90 0 0>}]
 dyn_sleeve_bott_in_r_01 [{shapeType=Box, extentX=6.6678, extentY=6.5805, extentZ=1.3177, center=<3.1448 0 0>}]
 dyn_sleeve_bott_in_r_02 [{shapeType=Box, extentX=6.6678, extentY=6.5805, extentZ=1.3177, center=<3.1448 0 0>}]
 dyn_sleeve_top_out_r [{shapeType=Capsule, length=3, radius=2, center=<0.2 0 0>, rotation=<rot 90 0 0>}]
 dyn_sleeve_top_r [{shapeType=Capsule, length=3, radius=2, center=<0.2 0 0>, rotation=<rot 90 0 0>}]
 hand_r [{shapeType=Capsule, length=2, radius=2, rotation=<rot 90 0 0>}]
 lowerarm_twist_01_r [{shapeType=Capsule, length=5.6492, radius=4.8853, center=<0.4939 0.1535 0.149>, rotation=<rot 0 -92.1844 -89.7169>}]
 upperarm_twist_01_r [{shapeType=Capsule, length=10.7343, radius=3.7522, center=<-23.5475 -0.0003 0.8774>, rotation=<rot 85 0 0>}]
 upperarm_twist_02_r [{shapeType=Capsule, length=2, radius=2, rotation=<rot 90 0 0>}]
}
constraints [
 {bone1Name=dyn_sleeve_bott_in_r_02, bone2Name=dyn_sleeve_bott_in_r_01, swing1LimitDegrees=20, swing2LimitDegrees=20, twistLimitDegrees=2}
 {bone1Name=dyn_sleeve_bott_in_r_01, bone2Name=upperarm_twist_02_r, swing1LimitDegrees=20, swing2LimitDegrees=20, twistLimitDegrees=2}
 {bone1Name=dyn_sleeve_bott_bk_r, bone2Name=upperarm_twist_01_r, swing1LimitDegrees=10, swing2LimitDegrees=20, twistLimitDegrees=2}
 {bone1Name=dyn_sleeve_bott_fr_r, bone2Name=upperarm_twist_01_r, swing1LimitDegrees=10, swing2LimitDegrees=20, twistLimitDegrees=2}
 {bone1Name=dyn_sleeve_top_r, bone2Name=clavicle_r, swing1LimitDegrees=10, swing2LimitDegrees=20, twistLimitDegrees=2}
 {bone1Name=dyn_sleeve_top_out_r, bone2Name=upperarm_twist_01_r, swing1LimitDegrees=10, swing2LimitDegrees=20, twistLimitDegrees=2}
 {bone1Name=dyn_gloove_cuff_r, bone2Name=lowerarm_twist_01_r, swing1LimitDegrees=20, swing2LimitDegrees=20, twistLimitDegrees=2}
 {
  bone1Name=hand_r
  bone2Name=clavicle_r
  swing1LimitDegrees=45
  swing1Motion=Free
  swing2LimitDegrees=45
  swing2Motion=Free
  twistLimitDegrees=45
  twistMotion=Free
 }
 {
  bone1Name=upperarm_twist_01_r
  bone2Name=clavicle_r
  swing1LimitDegrees=45
  swing1Motion=Free
  swing2LimitDegrees=45
  swing2Motion=Free
  twistLimitDegrees=45
  twistMotion=Free
 }
 {
  bone1Name=upperarm_twist_02_r
  bone2Name=clavicle_r
  swing1LimitDegrees=45
  swing1Motion=Free
  swing2LimitDegrees=45
  swing2Motion=Free
  twistLimitDegrees=45
  twistMotion=Free
 }
 {
  bone1Name=lowerarm_twist_01_r
  bone2Name=clavicle_r
  swing1LimitDegrees=45
  swing1Motion=Free
  swing2LimitDegrees=45
  swing2Motion=Free
  twistLimitDegrees=45
  twistMotion=Free
 }
]
