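# Physics asset for the rear belt on the male medium Robot Trouble character body, named as a belt fix for its dynamic bone.
# Two angled boxes form the dyn_rear_belt_mover body, alongside simple capsules on pelvis and spine_01.
# Both the belt body and spine_01 are constrained to the pelvis, with swing and twist explicitly free on the spine joint.
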PhysicsAsset /BRCosmetics/Characters/Player/Male/Medium/Bodies/M_MED_Robot_Trouble/Meshes/M_MED_Robot_Trouble_Dyn_BeltFix
bodies {
 dyn_rear_belt_mover [
  {shapeType=Box, extentX=3.7873, extentY=1.4742, extentZ=10.9536, rotation=<rot 0 0 -13.3555>}
  {shapeType=Box, extentX=3.7873, extentY=1.4742, extentZ=11.2394, center=<0.0197 2.13 -4.877>, rotation=<rot 0 0 -32.2899>}
 ]
 pelvis [{shapeType=Capsule, length=5, radius=5}]
 spine_01 [{shapeType=Capsule, length=4.98, radius=5, center=<8.4609 -2.694 0.4055>, rotation=<rot 0 0 74.8173>}]
}
constraints [
 {
  bone1Name=spine_01
  bone2Name=pelvis
  swing1LimitDegrees=45
  swing1Motion=Free
  swing2LimitDegrees=45
  swing2Motion=Free
  twistLimitDegrees=45
  twistMotion=Free
 }
 {bone1Name=dyn_rear_belt_mover, bone2Name=pelvis, swing1LimitDegrees=45, swing2LimitDegrees=45, twistLimitDegrees=45}
]
