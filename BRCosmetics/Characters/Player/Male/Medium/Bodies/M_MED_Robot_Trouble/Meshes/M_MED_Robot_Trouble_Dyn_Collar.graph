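# Physics asset for the dynamic collar of the male medium Robot Trouble character body.
# Multi-box bodies shape the back, left and right collar pieces, while capsules cover the head, spine_05 and two small collar master bones.
# The back piece constrains directly to spine_05, while each side piece constrains to its master with angular limits.
# The masters and head link to spine_05 with free swing and twist.
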PhysicsAsset /BRCosmetics/Characters/Player/Male/Medium/Bodies/M_MED_Robot_Trouble/Meshes/M_MED_Robot_Trouble_Dyn_Collar
bodies {
 dyn_collar_back [
  {shapeType=Box, extentX=10.9045, extentY=3.0132, extentZ=10, center=<6.6257 -2.0505 0>, rotation=<rot 0 -15.0035 0>}
  {shapeType=Box, extentX=9.6582, extentY=2.9279, extentZ=19.2144, center=<6.6413 -1.1299 -5.3082>, rotation=<rot -13.2844 -13.8766 -25.033>}
  {shapeType=Box, extentX=11.7002, extentY=3.2435, extentZ=16.35, center=<5.8196 -1.2062 5.316>, rotation=<rot 10.8334 -12.3472 33.0966>}
  {shapeType=Box, extentX=11.7002, extentY=3.2435, extentZ=10.5658, center=<4.8265 4.1667 11.7621>, rotation=<rot 18.4345 -3.5128 52.6944>}
 ]
 dyn_collar_l [
  {shapeType=Box, extentX=9.8051, extentY=10, extentZ=3, center=<6.1357 0.5726 -5.2365>, rotation=<rot -10.2413 1.988 8.331>}
  {shapeType=Box, extentX=9.8063, extentY=17.8354, extentZ=3.0725, center=<6.5451 -6.7781 -2.2677>, rotation=<rot -8.8808 -5.6135 40.8374>}
  {shapeType=Box, extentX=9.8051, extentY=10, extentZ=3, center=<5.4508 7.834 -5.4415>, rotation=<rot -15.8312 11.4222 -32.8233>}
 ]
 dyn_collar_l_MSTR [{shapeType=Capsule, length=3.3944, radius=1.2986, center=<3.149 -0.3339 -0.3662>, rotation=<rot 0 179.4939 -89.3007>}]
 dyn_collar_r [
  {shapeType=Box, extentX=11.157, extentY=10, extentZ=3, center=<5.6282 -0.0732 5.2056>, rotation=<rot 6.212 1.0282 -9.1078>}
  {shapeType=Box, extentX=11.2746, extentY=10.7331, extentZ=4.5, center=<5.9206 -3.4992 3.871>, rotation=<rot 16.106 -6.59 -24.3719>}
  {shapeType=Box, extentX=10.5881, extentY=10, extentZ=3, center=<4.7856 7.6943 4.6497>, rotation=<rot 14.1278 8.2942 22.5508>}
  {shapeType=Box, extentX=11.2746, extentY=8.8928, extentZ=4.5, center=<6.2942 -6.6703 1.6026>, rotation=<rot 16.106 -6.5908 -43.2551>}
 ]
 dyn_collar_r_MSTR [{shapeType=Capsule, length=3.2709, radius=1.0758, center=<2.5798 -0.3276 0.1796>, rotation=<rot 0 -1.482 -88.4832>}]
 head [
  {shapeType=Capsule, length=3.2545, radius=8.1288, center=<5.0344 2.632 -0.0214>, rotation=<rot 0.4599 -103.1591 89.8868>}
  {shapeType=Capsule, length=12.205, radius=4.29, center=<8.0111 -6.1883 -5.017>, rotation=<rot -2.2651 14.506 93.7421>}
  {shapeType=Capsule, length=11.491, radius=5.004, center=<8.574 -8.0897 4.5076>, rotation=<rot 2.434 14.4819 88.7924>}
  {shapeType=Capsule, length=9.7446, radius=6.7503, center=<8.9219 -6.3716 -0.5863>, rotation=<rot 0 21.8839 89.9997>}
 ]
 spine_05 [{shapeType=Capsule, length=5, radius=5}]
}
constraints [
 {
  bone1Name=head
  bone2Name=spine_05
  swing1LimitDegrees=45
  swing1Motion=Free
  swing2LimitDegrees=45
  swing2Motion=Free
  twistLimitDegrees=45
  twistMotion=Free
 }
 {bone1Name=dyn_collar_r, bone2Name=dyn_collar_r_MSTR, swing1LimitDegrees=75, swing2LimitDegrees=75, twistLimitDegrees=75}
 {bone1Name=dyn_collar_back, bone2Name=spine_05, swing1LimitDegrees=75, swing2LimitDegrees=75, twistLimitDegrees=75}
 {bone1Name=dyn_collar_l, bone2Name=dyn_collar_l_MSTR, swing1LimitDegrees=75, swing2LimitDegrees=75, twistLimitDegrees=90}
 {
  bone1Name=dyn_collar_l_MSTR
  bone2Name=spine_05
  swing1LimitDegrees=45
  swing1Motion=Free
  swing2LimitDegrees=45
  swing2Motion=Free
  twistLimitDegrees=45
  twistMotion=Free
 }
 {
  bone1Name=dyn_collar_r_MSTR
  bone2Name=spine_05
  swing1LimitDegrees=45
  swing1Motion=Free
  swing2LimitDegrees=45
  swing2Motion=Free
  twistLimitDegrees=45
  twistMotion=Free
 }
]
