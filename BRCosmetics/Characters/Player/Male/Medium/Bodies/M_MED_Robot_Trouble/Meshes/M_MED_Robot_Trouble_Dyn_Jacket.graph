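# Physics asset for the dynamic jacket of the M_MED_Robot_Trouble player cosmetic body, defining bodies and joints for its secondary motion.
# Seven two-bone capsule chains around the back, sides and front hang from spine_04, with limited swing and tight twist.
# Swing limits are asymmetric: back-right is stiffer, front-left is looser at the tip, and both front roots are narrow on one axis.
# Larger spine_04 capsules plus pelvis and spine_05 capsules form the torso, joined by angularly free constraints.
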PhysicsAsset /BRCosmetics/Characters/Player/Male/Medium/Bodies/M_MED_Robot_Trouble/Meshes/M_MED_Robot_Trouble_Dyn_Jacket
bodies {
 dyn_jacket_bk_01 [{shapeType=Capsule, length=3, radius=1, center=<1.5 0 0>, rotation=<rot 90 0 0>}]
 dyn_jacket_bk_02 [{shapeType=Capsule, length=3, radius=1, center=<1.5 0 0>, rotation=<rot 90 0 0>}]
 dyn_jacket_bk_l_01 [{shapeType=Capsule, length=3, radius=1, center=<1.5 0 0>, rotation=<rot 90 0 0>}]
 dyn_jacket_bk_l_02 [{shapeType=Capsule, length=3, radius=1, center=<1.5 0 0>, rotation=<rot 90 0 0>}]
 dyn_jacket_bk_r_01 [{shapeType=Capsule, length=3, radius=1, center=<-1.5 0 0>, rotation=<rot 90 0 0>}]
 dyn_jacket_bk_r_02 [{shapeType=Capsule, length=3, radius=1, center=<-1.5 0 0>, rotation=<rot 90 0 0>}]
 dyn_jacket_fr_l_01 [{shapeType=Capsule, length=3, radius=1, center=<3.6529 -0.0002 -0.4457>, rotation=<rot 90 0 0>}]
 dyn_jacket_fr_l_02 [{shapeType=Capsule, length=3, radius=1, center=<1.5 0 -0.5401>, rotation=<rot 90 0 0>}]
 dyn_jacket_fr_r_01 [{shapeType=Capsule, length=3, radius=1, center=<-1.5 0 0>, rotation=<rot 90 0 0>}]
 dyn_jacket_fr_r_02 [{shapeType=Capsule, length=3, radius=1, center=<-1.5 0 0>, rotation=<rot 90 0 0>}]
 dyn_jacket_l_01 [{shapeType=Capsule, length=3, radius=1, center=<1.5 0 0>, rotation=<rot 90 0 0>}]
 dyn_jacket_l_02 [{shapeType=Capsule, length=3, radius=1, center=<1.5 0 0>, rotation=<rot 90 0 0>}]
 dyn_jacket_r_01 [{shapeType=Capsule, length=3, radius=1, center=<-1.5 0 0>, rotation=<rot 90 0 0>}]
 dyn_jacket_r_02 [{shapeType=Capsule, length=3, radius=1, center=<-1.5 0 0>, rotation=<rot 90 0 0>}]
 pelvis [{shapeType=Capsule, length=3, radius=3}]
 spine_04 [
  {shapeType=Capsule, length=2.7655, radius=9.25, center=<2.6363 3.3858 0>}
  {shapeType=Capsule, length=3.75, radius=9.25, center=<0.3228 -0.457 0>}
  {shapeType=Capsule, length=0.285, radius=10.8941, center=<8.3337 2.6341 0>}
 ]
 spine_05 [{shapeType=Capsule, length=6.2125, radius=1.7109, center=<-4.7502 3.0285 -0.0515>, rotation=<rot 1.0137 -89.9999 0.3019>}]
}
constraints [
 {bone1Name=dyn_jacket_bk_02, bone2Name=dyn_jacket_bk_01, swing1LimitDegrees=45, swing2LimitDegrees=30, twistLimitDegrees=5}
 {bone1Name=dyn_jacket_bk_l_02, bone2Name=dyn_jacket_bk_l_01, swing1LimitDegrees=45, swing2LimitDegrees=30, twistLimitDegrees=5}
 {bone1Name=dyn_jacket_l_02, bone2Name=dyn_jacket_l_01, swing1LimitDegrees=45, swing2LimitDegrees=30, twistLimitDegrees=5}
 {bone1Name=dyn_jacket_fr_l_02, bone2Name=dyn_jacket_fr_l_01, swing1LimitDegrees=60, swing2LimitDegrees=60, twistLimitDegrees=5}
 {bone1Name=dyn_jacket_r_02, bone2Name=dyn_jacket_r_01, swing1LimitDegrees=45, swing2LimitDegrees=30, twistLimitDegrees=5}
 {bone1Name=dyn_jacket_fr_r_02, bone2Name=dyn_jacket_fr_r_01, swing1LimitDegrees=45, swing2LimitDegrees=30, twistLimitDegrees=5}
 {bone1Name=dyn_jacket_bk_r_02, bone2Name=dyn_jacket_bk_r_01, swing1LimitDegrees=20, swing2LimitDegrees=20, twistLimitDegrees=5}
 {bone1Name=dyn_jacket_bk_01, bone2Name=spine_04, swing1LimitDegrees=45, swing2LimitDegrees=30, twistLimitDegrees=5}
 {bone1Name=dyn_jacket_bk_l_01, bone2Name=spine_04, swing1LimitDegrees=45, swing2LimitDegrees=30, twistLimitDegrees=5}
 {bone1Name=dyn_jacket_l_01, bone2Name=spine_04, swing1LimitDegrees=45, swing2LimitDegrees=30, twistLimitDegrees=5}
 {bone1Name=dyn_jacket_fr_l_01, bone2Name=spine_04, swing1LimitDegrees=5, swing2LimitDegrees=60, twistLimitDegrees=5}
 {bone1Name=dyn_jacket_r_01, bone2Name=spine_04, swing1LimitDegrees=45, swing2LimitDegrees=30, twistLimitDegrees=5}
 {bone1Name=dyn_jacket_fr_r_01, bone2Name=spine_04, swing1LimitDegrees=5, swing2LimitDegrees=30, twistLimitDegrees=5}
 {bone1Name=dyn_jacket_bk_r_01, bone2Name=spine_04, swing1LimitDegrees=20, swing2LimitDegrees=20, twistLimitDegrees=5}
 {
  bone1Name=spine_04
  bone2Name=pelvis
  swing1LimitDegrees=45
  swing1Motion=Free
  swing2LimitDegrees=45
  swing2Motion=Free
  twistLimitDegrees=45
  twistMotion=Free
 }
 {
  bone1Name=spine_05
  bone2Name=spine_04
  swing1LimitDegrees=45
  swing1Motion=Free
  swing2LimitDegrees=45
  swing2Motion=Free
  twistLimitDegrees=45
  twistMotion=Free
 }
]
