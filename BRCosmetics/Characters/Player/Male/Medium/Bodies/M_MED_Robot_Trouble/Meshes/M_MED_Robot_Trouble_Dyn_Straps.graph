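# Physics asset for the dynamic chest, torso and back shoulder straps on the M_MED_Robot_Trouble male cosmetic body.
# Simple shapes on pelvis, upper spine, clavicles and upper-arm twist bones form the body framework, joined by angularly free constraints.
# Two-link chest and torso chains hang from spine_05 and spine_04 with limited swing and twist.
# Mirrored back shoulder chains use tighter limits below small MSTR root bodies, which attach freely to the clavicles.
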PhysicsAsset /BRCosmetics/Characters/Player/Male/Medium/Bodies/M_MED_Robot_Trouble/Meshes/M_MED_Robot_Trouble_Dyn_Straps
bodies {
 clavicle_l [{shapeType=Capsule, length=3, radius=2, rotation=<rot 90 0 0>}]
 clavicle_r [{shapeType=Capsule, length=3, radius=2, rotation=<rot 90 0 0>}]
 dyn_chest_strap_01 [{shapeType=Box, extentX=6.1104, extentY=1.3605, extentZ=7.5, center=<3.1297 0 -0.0001>}]
 dyn_chest_strap_02 [{shapeType=Box, extentX=6.1104, extentY=1.3605, extentZ=7.5, center=<3.1297 0 -0.0001>}]
 dyn_shoulder_strap_bk_l [{shapeType=Capsule, length=3, radius=2, center=<0.5151 0 0>, rotation=<rot 90 0 0>}]
 dyn_shoulder_strap_bk_l1 [{shapeType=Capsule, length=3, radius=2, center=<0.5151 0 0>, rotation=<rot 90 0 0>}]
 dyn_shoulder_strap_bk_l_MSTR [{shapeType=Capsule, length=0.505, radius=0.505, rotation=<rot 0 -90 0>}]
 dyn_shoulder_strap_bk_r [{shapeType=Capsule, length=3, radius=2, center=<-2.4844 -0.0001 0>, rotation=<rot 90 0 0>}]
 dyn_shoulder_strap_bk_r1 [{shapeType=Capsule, length=3, radius=2, center=<-2.4844 -0.0001 0>, rotation=<rot 90 0 0>}]
 dyn_shoulder_strap_bk_r_MSTR [{shapeType=Capsule, length=0.505, radius=0.505, rotation=<rot 0 -90 0>}]
 dyn_torso_strap_01 [{shapeType=Capsule, length=2, radius=3, center=<0.3 0 0>, rotation=<rot 90 0 0>}]
 dyn_torso_strap_02 [{shapeType=Capsule, length=2, radius=3, center=<0.3 0 0>, rotation=<rot 90 0 0>}]
 pelvis [{shapeType=Capsule, length=3, radius=3}]
 spine_04 [
  {shapeType=Capsule, length=2.7655, radius=9.25, center=<2.6363 3.3858 0>}
  {shapeType=Capsule, length=3.75, radius=9.25, center=<0.3228 -0.457 0>}
  {shapeType=Capsule, length=1.1759, radius=10.8941, center=<8.3337 2.6341 0>}
 ]
 spine_05 [{shapeType=Capsule, length=6.2125, radius=1.7109, center=<-4.7502 3.0285 -0.0515>, rotation=<rot 1.0137 -89.9999 0.3019>}]
 upperarm_twist_01_l [{shapeType=Capsule, length=1, radius=1}]
 upperarm_twist_01_r [{shapeType=Capsule, length=1, radius=1}]
}
constraints [
 {bone1Name=dyn_shoulder_strap_bk_r, bone2Name=dyn_shoulder_strap_bk_r_MSTR, swing1LimitDegrees=5, swing2LimitDegrees=30, twistLimitDegrees=1}
 {bone1Name=dyn_shoulder_strap_bk_l, bone2Name=dyn_shoulder_strap_bk_l_MSTR, swing1LimitDegrees=5, swing2LimitDegrees=30, twistLimitDegrees=1}
 {
  bone1Name=clavicle_l
  bone2Name=spine_05
  swing1LimitDegrees=45
  swing1Motion=Free
  swing2LimitDegrees=45
  swing2Motion=Free
  twistLimitDegrees=45
  twistMotion=Free
 }
 {
  bone1Name=clavicle_r
  bone2Name=spine_05
  swing1LimitDegrees=45
  swing1Motion=Free
  swing2LimitDegrees=45
  swing2Motion=Free
  twistLimitDegrees=45
  twistMotion=Free
 }
 {
  bone1Name=spine_04
  bone2Name=pelvis
  swing1LimitDegrees=45
  swing1Motion=Free
  swing2LimitDegrees=45
  swing2Motion=Free
  twistLimitDegrees=45
  twistMotion=Free
 }
 {bone1Name=dyn_shoulder_strap_bk_l1, bone2Name=dyn_shoulder_strap_bk_l, swing1LimitDegrees=5, swing2LimitDegrees=30, twistLimitDegrees=5}
 {bone1Name=dyn_shoulder_strap_bk_r1, bone2Name=dyn_shoulder_strap_bk_r, swing1LimitDegrees=5, swing2LimitDegrees=30, twistLimitDegrees=5}
 {
  bone1Name=spine_05
  bone2Name=spine_04
  swing1LimitDegrees=45
  swing1Motion=Free
  swing2LimitDegrees=45
  swing2Motion=Free
  twistLimitDegrees=45
  twistMotion=Free
 }
 {bone1Name=dyn_chest_strap_01, bone2Name=spine_05, swing1LimitDegrees=45, swing2LimitDegrees=2, twistLimitDegrees=2}
 {bone1Name=dyn_chest_strap_02, bone2Name=dyn_chest_strap_01, swing1LimitDegrees=45, swing2LimitDegrees=45, twistLimitDegrees=10}
 {bone1Name=dyn_torso_strap_01, bone2Name=spine_04, swing1LimitDegrees=10, swing2LimitDegrees=15, twistLimitDegrees=5}
 {bone1Name=dyn_torso_strap_02, bone2Name=dyn_torso_strap_01, swing1LimitDegrees=10, swing2LimitDegrees=15, twistLimitDegrees=5}
 {
  bone1Name=upperarm_twist_01_l
  bone2Name=clavicle_l
  swing1LimitDegrees=45
  swing1Motion=Free
  swing2LimitDegrees=45
  swing2Motion=Free
  twistLimitDegrees=45
  twistMotion=Free
 }
 {
  bone1Name=upperarm_twist_01_r
  bone2Name=clavicle_r
  swing1LimitDegrees=45
  swing1Motion=Free
  swing2LimitDegrees=45
  swing2Motion=Free
  twistLimitDegrees=45
  twistMotion=Free
 }
 {
  bone1Name=dyn_shoulder_strap_bk_l_MSTR
  bone2Name=clavicle_l
  swing1LimitDegrees=45
  swing1Motion=Free
  swing2LimitDegrees=45
  swing2Motion=Free
  twistLimitDegrees=45
  twistMotion=Free
 }
 {
  bone1Name=dyn_shoulder_strap_bk_r_MSTR
  bone2Name=clavicle_r
  swing1LimitDegrees=45
  swing1Motion=Free
  swing2LimitDegrees=45
  swing2Motion=Free
  twistLimitDegrees=45
  twistMotion=Free
 }
]
